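# Niagara anim trail effect for the Jazz Shoes emote, drawing a short-lived dark grey ribbon trail from a skeletal mesh.
# A single emitter, Trail_Smoke002, uses shared AnimTrail modules to spawn, initialize from bone or socket indices, and update particles.
# Ribbon width tapers late in life while alpha fades in and out, feeding two ribbon renderers.
# The DynamicMaterialParameters module is present but disabled.
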NiagaraSystem /BRCosmetics/Animation/Game/MainPlayer/Emotes/JazzShoes/FX/NS_Emote_JazzShoes_Trail
dependencies {usedDataInterfaces [&NiagaraDataInterfaceCurve], usedDynamicInputs [], usedModules [&SystemState]}
emitters {Trail_Smoke002 {
  emitterUpdateScript [
   &EmitterState()
   &FN_SpawnAnimTrail_v2()
  ]
  particleSpawnScript [
   /Game/Effects/Niagara/Modules/AnimTrail/FN_InitializeAnimTrail/FN_InitializeAnimTrailBoneIndecies(First Bone or Socket Index=1)
   /Game/Effects/Niagara/Modules/AnimTrail/FN_InitializeAnimTrail/FN_InitializeAnimTrail(AnimTrailWidth=1, Color=<rgba 0.0247 0.0247 0.0247 1>, Lifetime=0.25)
  ]
  particleUpdateScript [
   &FN_UpdateAnimTrail_v2()
   &ScaleRibbonWidth(Ribbon Width Scale=&FloatFromCurve(FloatCurve=DataInterface() {propertyValues {Curve {keys [
         {tangentMode=RCTM_User, time=0, value=1}
         {arriveTangent=-0.0405, interpMode=RCIM_Cubic, leaveTangent=-0.0405, tangentMode=RCTM_User, time=0.7009, value=1}
         {arriveTangent=-8.0089, interpMode=RCIM_Cubic, leaveTangent=-8.0089, tangentMode=RCTM_User, time=1, value=0}
        ]}}})) {module=ScaleRibbonWidth001}
   &SolveForcesAndVelocity(Clamp Velocity=false, Limit Acceleration=false)
   &DynamicMaterialParameters(
    Index 0 Param 1=&RandomRangeFloat(Minimum=-0.25, Maximum=0.25)
    Index 0 Param 2=5
    Index 0 Param 3=-0.8
    Index 0 Param 4=&FloatFromCurve(FloatCurve=DataInterface() {propertyValues {Curve {keys [
         {interpMode=RCIM_Cubic, tangentMode=RCTM_User, time=0, value=1}
         {arriveTangent=-0.6219, interpMode=RCIM_Cubic, leaveTangent=-0.6219, tangentMode=RCTM_User, time=1, value=0}
        ]}}})
    Module.Param0WriteEnabled=false
    Module.Param1WriteEnabled=true
    Module.Param2WriteEnabled=false
    Module.Param3WriteEnabled=false
   ) {enabled=false}
   &ScaleColor(
    Scale Alpha=&FloatFromCurve(FloatCurve=DataInterface() {propertyValues {Curve {keys [
         {interpMode=RCIM_Cubic, tangentMode=RCTM_User, time=0, value=0}
         {interpMode=RCIM_Cubic, tangentMode=RCTM_User, time=0.5, value=1}
         {tangentMode=RCTM_User, time=1, value=0}
        ]}}})
    ScaleA=true
    ScaleRGB=true
   )
  ]
  rendererClasses [&NiagaraRibbonRendererProperties]
  renderers [
   {rendererClass=&NiagaraRibbonRendererProperties}
   {rendererClass=&NiagaraRibbonRendererProperties, rendererIndex=1}
  ]
 }}
summary {
 systemName=NS_Emote_JazzShoes_Trail
 emitters [{emitterName=Trail_Smoke002, rendererClasses [&NiagaraRibbonRendererProperties]}]
 userVariables [
  User.Foo {type=NiagaraFloat, defaultValue=0, description=""}
  User.FXTrailScale {type=NiagaraFloat, defaultValue=0, description=""}
  User.RibbonIDOffset {type=NiagaraInt32}
  User.SkeletalMesh {
   type=NiagaraDataInterfaceSkeletalMesh
   defaultValue {
    struct=NiagaraExt_VariableValue_DataInterface
    value {dataInterface=&self:NiagaraDataInterfaceSkeletalMesh_0}
   }
  }
 ]
}
